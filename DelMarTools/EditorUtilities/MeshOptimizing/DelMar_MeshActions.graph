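# Editor utility for DelMar tools that turns off shadow casting on the static meshes selected in the editor.
# DisableShadowsForLODs filters the selection to StaticMesh assets and passes each to internal_disableShadows.
# That helper uses StaticMeshEditorSubsystem to disable cast shadow on every section of each LOD not listed in PreserveLODs.
# It then zeroes LOD 0's distance field resolution scale, regardless of the preserved list.
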EditorUtilityBlueprint /DelMarTools/EditorUtilities/MeshOptimizing/DelMar_MeshActions
graphs [
 DisableShadowsForLODs {nodes [
   DisableShadowsForLODs!() -> (then, PreserveLODs:Array of Integers) {is=title K2Node_FunctionEntry}
   > n1 = GetSelectedAssets()
   > n2 = Filter by Class!(
    execute
    TargetArray:Array of Object References=@n1
    ObjectClass:Object Class Reference=&StaticMesh
    FilterType:EEditorScriptingFilterType Enum
   ) -> (then, ReturnValue:Array of Static Mesh Object References)
   > n3 = For Each Loop(Array:Array of Static Mesh Object References=@n2) -> (Array Element:Static Mesh Object Reference)
   > internal_disableShadows!(
    execute
    self
    staticMesh:Static Mesh Object Reference=@n3
    PreserveLODs:Array of Integers=Get PreserveLODs() -> (PreserveLODs:Array of Integers)
   ) -> (then)
  ]}
 EventGraph {links [], nodes []}
 internal_disableShadows {nodes [
   n1 = internal_disableShadows!() -> (then, staticMesh:Static Mesh Object Reference, PreserveLODs:Array of Integers) {is=title K2Node_FunctionEntry}
   > Set l_mesh(l_mesh:Static Mesh Object Reference=@n1) -> (Output_Get:Static Mesh Object Reference)
   > n2 = Sequence()
   n3 = Get StaticMeshEditorSubsystem()
   > n4 = GetLodCount!(
    execute
    self:Static Mesh Editor Subsystem Object Reference=@n3
    StaticMesh:Static Mesh Object Reference=Get l_mesh() -> (l_mesh:Static Mesh Object Reference)
   ) -> (then, ReturnValue:int=0)
   > n5 = For Loop(LastIndex=int - int(A=@n4, B=1) {is=title K2Node_CallFunction})
   n6 = Get PreserveLODs()
   > Branch(Condition=NOT Boolean(A=Contains Item(TargetArray:Array of Integers=@n6, ItemToFind:"Integer (by ref)"=@n5)))
   n7 = Get l_mesh()
   > n8 = For Loop(LastIndex=int - int(A=GetNumSections!(self:Static Mesh Object Reference=@n7, InLOD:int=@n5) -> (ReturnValue:int=0), B=1))
   > EnableSectionCastShadow!(
    execute
    self:Static Mesh Editor Subsystem Object Reference=Get StaticMeshEditorSubsystem()
    StaticMesh:Static Mesh Object Reference=Get l_mesh()
    bCastShadow:bool=false
    LODIndex:int=@n5
    SectionIndex:int=@n8
   ) -> (then)
   n9 = GetLodBuildSettings!(
    execute=@n2.then_1
    self:Static Mesh Editor Subsystem Object Reference=@n3
    StaticMesh:Static Mesh Object Reference=Get l_mesh()
    LodIndex:int=0
   ) -> (then, OutBuildOptions:Mesh Build Settings Structure)
   > n10 = Set members in Mesh Build Settings!(execute:Exec, StructRef:"Mesh Build Settings Structure (by ref)"=@n9, DistanceFieldResolutionScale:float=0) -> (then:Exec, StructOut:"Mesh Build Settings Structure (by ref)") {is=title K2Node_SetFieldsInStruct}
   > SetLodBuildSettings!(
    execute
    self:Static Mesh Editor Subsystem Object Reference=@n3
    StaticMesh:Static Mesh Object Reference=Get l_mesh()
    LodIndex:int=0
    BuildOptions:"Mesh Build Settings Structure (by ref)"=@n10
   ) -> (then)
  ]}
]
variables []
